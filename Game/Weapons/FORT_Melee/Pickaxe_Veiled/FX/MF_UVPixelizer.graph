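# MF_UVPixelizer is a library-exposed material function in the Veiled pickaxe FX folder that pixelates UV coordinates.
# It multiplies the UV input by a scalar Resolution, rounds the result, then divides by Resolution again, snapping coordinates to a uniform grid.
# The UV input previews with texture coordinates, and rounding to nearest, rather than flooring, centers each cell on its grid point.
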MaterialFunction /Game/Weapons/FORT_Melee/Pickaxe_Veiled/FX/MF_UVPixelizer
graph {
 properties {bExposeToLibrary=true}
 nodes [
  n1 = FunctionInput() {inputName=Resolution, inputType=FunctionInput_Scalar, previewValue=<10 0 0 1>}
  n2 = Multiply(A=FunctionInput(Preview=TextureCoordinate()) {inputName=UV, inputType=FunctionInput_Vector2}, B=@n1)
  FunctionOutput(None=Divide(A=Round(None=@n2), B=@n1))
 ]
}
